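# Translucent, two-sided Niagara sprite material that draws a square pixel particle for the Pencil Cherry pickaxe FX.
# Two SquarePatternFromUVs masks, one sized by a parameter and one fixed, are summed after offsetting the fixed one down, giving a two-tone square.
# The sum multiplies particle color for tint, while its Ceil multiplies particle alpha to form a hard-edged mask.
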Material /BRCosmetics/Weapons/FORT_Melee/Pickaxe_PencilCherry/FX/M_PencilCherry_PixelSquare
graph {
 properties {bUsedWithNiagaraSprites=true, blendMode=BLEND_Translucent, pixelDepthOffsetMode=PDOM_Legacy, twoSided=true, usedShadingModels=Default Lit}
 nodes [
  n1 = TextureCoordinate()
  n2 = Constant() {r=0.4}
  n3 = Add(
   A=MaterialFunctionCall!(UVs=TextureCoordinate(), Size=ScalarParameter(), SizeScale) {materialFunction=/Game/Packages/Fortress_SharedMaterials/Base_Material_Functions/SquarePatternFromUVs}.PatternOutput
   B=Subtract(A=MaterialFunctionCall(UVs=@n1, Size=@n2) {materialFunction=&SquarePatternFromUVs}.PatternOutput) {constB=0.5}
  )
  n4 = ParticleColor()
  Multiply(A=@n3, B=@n4.RGB)
  Multiply()
  Multiply(A=@n4.A, B=Ceil(None=@n3))
 ]
}
