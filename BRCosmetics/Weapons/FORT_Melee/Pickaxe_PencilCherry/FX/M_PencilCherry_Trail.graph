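# Translucent two-sided Niagara ribbon material for the PencilCherry pickaxe trail.
# Trail texture sampled in object-stabilized screen-space UVs supplies a pattern, shaped by length-edge fades and a trailing-edge sphere mask.
# A second texture channel and the dynamic Erosion parameter are subtracted, then Ceil cuts a hard-edged mask.
# That mask gates particle-tinted color and an alpha term with leading/trailing fades, near-camera fade and particle alpha.
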Material /BRCosmetics/Weapons/FORT_Melee/Pickaxe_PencilCherry/FX/M_PencilCherry_Trail
graph {
 properties {bUsedWithNiagaraRibbons=true, blendMode=BLEND_Translucent, pixelDepthOffsetMode=PDOM_Legacy, twoSided=true, usedShadingModels=Default Lit}
 nodes [
  MaterialFunctionCall!(HSV_In) {materialFunction=/Game/Athena/UI/Materials/MF_HSVtoRGB}
  n1 = ComponentMask(None=TextureCoordinate()) {r=true}
  n2 = Power(
   Base=Add(A=OneMinus(None=@n1), B=ScalarParameter() {defaultValue=0.4, parameterName=Pixel_Offset})
   Exp=ScalarParameter() {defaultValue=4, parameterName=Pixel_Power}
  )
  n3 = TextureSample(UVs=Multiply(
    A=MaterialFunctionCall!(Position, UVRadialScale) {materialFunction=/Game/Weapons/WeaponSkins/MaterialFunction/ObjectStabilizedScreenSpaceUVs}.StabilizedUVs
    B=ScalarParameter() {defaultValue=30, parameterName=UV_Scale}
   )) {texture=&/T_PencilCherryTrail}
  n4 = ComponentMask(None=TextureCoordinate()) {g=true}
  n5 = Add(A=@n4, B=ScalarParameter() {group="Fade Top && Bottom Length Edges", parameterName=Top Length Fade Bias, sortPriority=2})
  n6 = ScalarParameter() {defaultValue=4, group="Fade Top && Bottom Length Edges", parameterName=Top Length Fade Scale}
  n7 = LinearInterpolate(
   B=Saturate(None=Multiply(A=@n5, B=@n6) {constB=3})
   Alpha=ScalarParameter() {defaultValue=1, group=Alpha, parameterName=Fade Top Length Edge, sortPriority=7}
  ) {constA=1}
  n8 = Add(A=OneMinus(None=@n4), B=ScalarParameter() {group="Fade Top && Bottom Length Edges", parameterName=Bottom Length Fade Bias, sortPriority=3})
  n9 = ScalarParameter() {defaultValue=4, group="Fade Top && Bottom Length Edges", parameterName=Bottom Length Fade Scale, sortPriority=1}
  n10 = LinearInterpolate(
   B=Saturate(None=Multiply(A=@n8, B=@n9) {constB=2})
   Alpha=ScalarParameter() {defaultValue=1, group=Alpha, parameterName=Fade Bottom Length Edge, sortPriority=8}
  ) {constA=1}
  n11 = TextureCoordinate()
  n12 = VectorParameter() {group=Trailing Edge Sphere Mask, parameterName=Trailing Edge Sphereical UV Scale, sortPriority=4, defaultValue=<rgba 0.5 1 0 0>}
  n13 = ComponentMask(None=@n12.RGB) {g=true, r=true}
  n14 = VectorParameter() {group=Trailing Edge Sphere Mask, parameterName="Trailing Edge Sphere Mask Coord (B)", sortPriority=3, defaultValue=<rgba 0.85 0.82 0 0>}
  n15 = OneMinus(None=SphereMask(
    A=Multiply(A=@n11, B=@n13)
    B=ComponentMask(None=@n14.RGB) {g=true, r=true}
    Radius=ScalarParameter() {defaultValue=0.8, group=Trailing Edge Sphere Mask, parameterName=Trailing Edge Radius, sortPriority=1}
    Hardness=ScalarParameter() {defaultValue=0.3, group=Trailing Edge Sphere Mask, parameterName=Trailing Edge Hardness, sortPriority=2}
   ))
  n16 = ScalarParameter() {defaultValue=5, group=Trailing Edge Sphere Mask, parameterName=Trailing Edge Contrast}
  n17 = Multiply(
   A=Multiply(A=Multiply(A=@n2, B=@n3.G), B=Multiply(A=@n7, B=@n10))
   B=LinearInterpolate(
    B=Power(Base=@n15, Exp=@n16) {constExponent=5}
    Alpha=ScalarParameter() {defaultValue=1, group=Alpha, parameterName=Multiply Trailing Edge Sphere Mask, sortPriority=6}
   ) {constA=1}
  )
  n18 = MaterialFunctionCall(
   UVs=TextureCoordinate()
   Texture Scale=MaterialFunctionCall(R=ScalarParameter() {defaultValue=1, parameterName=Param_2}, G=ScalarParameter() {defaultValue=0.5, parameterName=Param_1}) {materialFunction=&AppendMany}.RG
  ) {materialFunction=&ScaleUVsByCenter}
  n19 = AppendVector(A=ScalarParameter() {defaultValue=-0.02}, B=Constant())
  n20 = LinearInterpolate(
   B=TextureSample(UVs=Add(A=@n18.UVs, B=@n19)) {texture=&T_PencilCherryTrail}.B
   Alpha=ScalarParameter() {defaultValue=1, parameterName=Param_3}
  )
  n21 = Ceil(None=Subtract(A=Subtract(A=@n17, B=@n20), B=DynamicParameter() {defaultValue=<rgba 0.1 1 1 1>, paramNames [Erosion,Param2,Param3,Param4]}.Erosion))
  n22 = ComponentMask(None=TextureCoordinate()) {r=true}
  n23 = Multiply(
   A=Add(A=@n22, B=ScalarParameter() {group="Fade Leading && Trailing Edges", parameterName=Leading Edge Fade Bias, sortPriority=2})
   B=ScalarParameter() {defaultValue=10, group="Fade Leading && Trailing Edges", parameterName=Leading Edge Fade Scale}
  ) {constB=3}
  n24 = Multiply(
   A=Add(
    A=OneMinus(None=@n22)
    B=ScalarParameter() {defaultValue=-0.4, group="Fade Leading && Trailing Edges", parameterName=Trailing Edge Fade Bias, sortPriority=3}
   )
   B=ScalarParameter() {defaultValue=4, group="Fade Leading && Trailing Edges", parameterName=Trailing Edge Fade Scale, sortPriority=1}
  ) {constB=2}
  n25 = Multiply(
   A=LinearInterpolate(B=Saturate(None=@n23), Alpha=ScalarParameter() {defaultValue=1, group=Alpha, parameterName=Fade Leading Edge, sortPriority=4}) {constA=1}
   B=LinearInterpolate(B=Saturate(None=@n24), Alpha=ScalarParameter() {defaultValue=0.8, group=Alpha, parameterName=Fade Trailing Edge, sortPriority=5}) {constA=1}
  )
  n26 = ScalarParameter() {defaultValue=64, group=Alpha, parameterName=NearFadeDistance, sortPriority=3}
  Multiply(
   A=Multiply(
    A=Multiply(A=@n21, B=@n25)
    B=MaterialFunctionCall(Distance=@n26) {materialFunction=&Near_Camera_Fade}.Result
   )
   B=ParticleColor().A
  )
  Multiply(A=Multiply(A=ConstantBiasScale(None=@n3.R) {scale=0.2}, B=ParticleColor().RGB), B=@n21)
  n27 = TextureCoordinate()
  NamedRerouteDeclaration(None=OneMinus(None=ComponentMask(None=@n27) {r=true})) {name=RibbonUVs, nodeColor=<rgba 1 0.7266 0 1>}
  RgbToHsv!(None)
 ]
}
