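# AnimBlueprint for the PencilCherry pickaxe that offsets its left and right axe bones according to the wielder's character mesh.
# Each update casts the pawn owner to FortPlayerPawn and searches the part mesh's display name for "LRG", setting larger offsets when found and smaller ones otherwise.
# The AnimGraph applies these variables to Axe_L and Axe_R through additive bone-space Modify Bone nodes, with mirrored rotations and a shared translation.
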AnimBlueprint /BRCosmetics/Weapons/FORT_Melee/Pickaxe_PencilCherry/Meshes/PencilCherry_Axe_AnimBP
target_skeleton=&/PencilCherry_Axe_Skeleton
graphs [
 AnimGraph {nodes [
   n1 = "Transform (Modify) Bone - Bone: Axe_L"(
    ComponentPose=Local To Component(LocalPose=Input Pose())
    Alpha=1
    Translation=Get Pos Offset() -> (Pos Offset:Vector)
    Rotation=Get Rot Offset Left() -> (Rot Offset Left:Rotator)
    Scale="1,1,1"
   ) {
    rotationMode=BMM_Additive
    rotationSpace=BCS_BoneSpace
    translationMode=BMM_Additive
    translationSpace=BCS_BoneSpace
    alphaBoolBlend {blendOption=Linear}
    boneToModify {boneName=Axe_L}
   }
   n2 = "Transform (Modify) Bone - Bone: Axe_R"(ComponentPose=@n1, Alpha=1, Translation=Get Pos Offset(), Rotation=Get Rot Offset Right() -> (Rot Offset Right:Rotator), Scale="1,1,1") {
    rotationMode=BMM_Additive
    rotationSpace=BCS_BoneSpace
    translationMode=BMM_Additive
    translationSpace=BCS_BoneSpace
    alphaBoolBlend {blendOption=Linear}
    boneToModify {boneName=Axe_R}
   }
   Output Pose(Result=Component To Local(ComponentPose=@n2)) {layerGroup=None}
  ]}
 EventGraph {nodes [
   Event BlueprintUpdateAnimation()
   > n1 = Cast To FortPlayerPawn(Object=TryGetPawnOwner())
   n2 = GetDisplayName(Object=GetSkeletalMeshForPartType(self=@n1))
   > n3 = Branch(Condition="integer >= integer"(A=FindSubstring(SearchIn=@n2, Substring=LRG)))
   > Set Pos Offset(Pos Offset:Vector="1,0,0") -> (Output_Get:Vector)
   > Set Rot Offset Left(Rot Offset Left:Rotator="0,10,0") -> (Output_Get:Rotator)
   > Set Rot Offset Right(Rot Offset Right:Rotator="0,-10,0") -> (Output_Get:Rotator)
   Set Pos Offset(execute=@n3.else, Pos Offset="0, 0, 0")
   > Set Rot Offset Left(Rot Offset Left="0,5,0")
   > Set Rot Offset Right(Rot Offset Right="0,-5,0")
  ]}
]
variables [
 Pos Offset {type=Vector}
 Rot Offset Left {type=Rotator}
 Rot Offset Right {type=Rotator}
]
